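; Gameplay ability for a Summer Event quest voiceover (Q03 VO1) that manages an audio mix while it runs.
; On activation it calls the parent ability, then pushes a music low-pass/wind quest sound mix modifier.
; The mix is popped when the ability ends, and also if passive activation fails.
; No audio playback logic appears here, so any voiceover playback would come from the parent class.
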(event Ability|EventActivateAbility
  (|CallParentFunction)
  (Audio|PushSoundMixModifier "/Game/Sounds/FortSoundClassesAndModes/Fort_Music_Quest_LPF_Wind.Fort_Music_Quest_LPF_Wind"))

(event Ability|EventOnEndAbility
  (params (bWasCancelled :type Boolean))
  (|CallParentFunction bWasCancelled)
  (Audio|PopSoundMixModifier "/Game/Sounds/FortSoundClassesAndModes/Fort_Music_Quest_LPF_Wind.Fort_Music_Quest_LPF_Wind"))

(event Ability|EventFailedToActivatePassiveAbility
  (params (ActorInfo :type Gameplay%20Ability%20Actor%20Info%20Structure))
  (Audio|PopSoundMixModifier "/Game/Sounds/FortSoundClassesAndModes/Fort_Music_Quest_LPF_Wind.Fort_Music_Quest_LPF_Wind"))
